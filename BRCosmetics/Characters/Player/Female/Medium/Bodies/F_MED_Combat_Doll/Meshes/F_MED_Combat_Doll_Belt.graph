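# Physics asset for the belt mesh of the F_MED_Combat_Doll female player cosmetic body, defining collision and joint limits for its bones.
# Box bodies cover the left and right belt segments and a dyn_belt3 piece, alongside a small capsule on the pelvis.
# Each belt body is constrained directly to the pelvis rather than chained, with narrow swing and twist limits that allow only slight movement, widest on one swing axis of belt1_l.
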PhysicsAsset /BRCosmetics/Characters/Player/Female/Medium/Bodies/F_MED_Combat_Doll/Meshes/F_MED_Combat_Doll_Belt
bodies {
 belt1_l [{shapeType=Box, extentX=2, extentY=5, extentZ=8.7628, rotation=<rot 0 90 0>}]
 belt1_r [{shapeType=Box, extentX=2, extentY=5, extentZ=7.3206, rotation=<rot 0 -90 0>}]
 belt2_l [{shapeType=Box, extentX=2, extentY=5, extentZ=7.2506, rotation=<rot 0 90 0>}]
 belt2_r [{shapeType=Box, extentX=2, extentY=6.4617, extentZ=5, rotation=<rot -42.8883 89.9998 89.9998>}]
 dyn_belt3 [{shapeType=Box, extentX=5, extentY=3, extentZ=9}]
 pelvis [{shapeType=Capsule, length=1, radius=1}]
}
constraints [
 {bone1Name=belt1_l, bone2Name=pelvis, swing1LimitDegrees=5, swing2LimitDegrees=30, twistLimitDegrees=5}
 {bone1Name=belt2_l, bone2Name=pelvis, swing1LimitDegrees=15, swing2LimitDegrees=15, twistLimitDegrees=5}
 {bone1Name=belt1_r, bone2Name=pelvis, swing1LimitDegrees=15, swing2LimitDegrees=15, twistLimitDegrees=5}
 {bone1Name=belt2_r, bone2Name=pelvis, swing1LimitDegrees=15, swing2LimitDegrees=15, twistLimitDegrees=5}
 {bone1Name=dyn_belt3, bone2Name=pelvis, swing1LimitDegrees=15, swing2LimitDegrees=15, twistLimitDegrees=5}
]
