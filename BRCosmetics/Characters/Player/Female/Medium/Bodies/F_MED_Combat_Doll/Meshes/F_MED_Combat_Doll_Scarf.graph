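# Physics asset for the scarf mesh of the female medium Combat Doll player cosmetic body.
# It defines two capsule bodies, one on scarf_tail and a small one on spine_05.
# A single constraint joins the tail to the spine, with tight degree limits set on both swings and twist to restrict how far the tail can move.
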PhysicsAsset /BRCosmetics/Characters/Player/Female/Medium/Bodies/F_MED_Combat_Doll/Meshes/F_MED_Combat_Doll_Scarf
bodies {
 scarf_tail [{shapeType=Capsule, length=2.8182, radius=1.6853, center=<-1.6944 0.19 0>, rotation=<rot 90 0 0>}]
 spine_05 [{shapeType=Capsule, length=1, radius=1}]
}
constraints [{bone1Name=scarf_tail, bone2Name=spine_05, swing1LimitDegrees=8, swing2LimitDegrees=15, twistLimitDegrees=5}]
