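# Secondary-motion AnimBP for the F_MED_Combat_Doll body, layered on the pose copied from the attached parent mesh.
# Scarf RigidBody feeds AnimDynamics belt springs, then additive bone corrections for belts, buckle, thigh volume and scarf back, plus half-weight thigh rotation constraints.
# EventGraph maps thigh and head ref-pose deltas to corrective alphas.
# The belt RigidBody, weighted by crouch and speed, is disconnected from output; scarf_front_alpha is unused.
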AnimBlueprint /BRCosmetics/Characters/Player/Female/Medium/Bodies/F_MED_Combat_Doll/Meshes/F_MED_Combat_Doll_AnimBP
target_skeleton=&/F_MED_Combat_Doll_Skeleton
graphs [
 AnimGraph {nodes [
   RigidBody(Alpha=Get belt_alpha() -> (belt_alpha:double)) {
    bClampLinearTranslationLimitToRefPose=true
    bOverrideWorldGravity=true
    overridePhysicsAsset=&/F_MED_Combat_Doll_Belt
    simulationSpace=BaseBoneSpace
    alphaBoolBlend {blendOption=Linear}
    baseBoneRef {boneName=pelvis}
    componentAppliedLinearAccClamp=<900 900 3000>
    componentLinearAccScale=<0.3 0.3 0.3>
    componentLinearVelScale=<0.3 0.3 0.3>
    overrideWorldGravity {z=-980}
    simSpaceSettings {externalAngularVelocity {z=1}, externalLinearVelocity {z=1}}
   }
   n1 = RigidBody(ComponentPose=Local To Component(LocalPose=Copy Pose From Mesh() {bUseAttachedParent=true})) {
    bClampLinearTranslationLimitToRefPose=true
    bOverrideWorldGravity=true
    overridePhysicsAsset=&/F_MED_Combat_Doll_Scarf
    simulationSpace=BaseBoneSpace
    alphaBoolBlend {blendOption=Linear}
    baseBoneRef {boneName=root}
    componentAppliedLinearAccClamp=<900 900 900>
    componentLinearAccScale=<0.3 0.3 0.3>
    componentLinearVelScale=<0.3 0.3 0.3>
    overrideWorldGravity {z=-600}
   }
   n2 = Anim Dynamics - Bone: dyn_belt3(ComponentPose=@n1, Alpha=1, ExternalForce="0,0,0") {
    angularDampingOverride=0.999
    angularSpringConstant=15000
    bGravityOverrideInSimSpace=true
    bLinearSpring=true
    bOverrideAngularDamping=true
    gravityScale=1.01
    linearSpringConstant=50
    numSolverIterationsPostUpdate=4
    numSolverIterationsPreUpdate=16
    simulationSpace=RootRelative
    alphaBoolBlend {blendOption=Linear}
    boundBone {boneName=dyn_belt3}
    componentAppliedLinearAccClamp=<3000 3000 3000>
    componentLinearAccScale {z=1}
    componentLinearVelScale {z=0.5}
    gravityOverride {z=-980}
    physicsBodyDefinitions [{
      sphereCollisionRadius=0
      boundBone {boneName=dyn_belt3}
      boxExtents=<90 45 45>
      constraintSetup {angularLimitsMax=<0 0 5>, angularLimitsMin=<0 0 -10>, angularTarget=<1 0 0>, linearAxesMin=<-3 0 0>}
      localJointOffset=<1 0 0>
     }]
   }
   n3 = Anim Dynamics - Bone: belt2_r(ComponentPose=@n2, Alpha=1, ExternalForce="0,0,0") {
    angularDampingOverride=0.999
    angularSpringConstant=15000
    bGravityOverrideInSimSpace=true
    bLinearSpring=true
    bOverrideAngularDamping=true
    gravityScale=1.01
    linearSpringConstant=50
    numSolverIterationsPostUpdate=4
    numSolverIterationsPreUpdate=16
    simulationSpace=RootRelative
    alphaBoolBlend {blendOption=Linear}
    boundBone {boneName=belt2_r}
    componentAppliedLinearAccClamp=<3000 3000 3000>
    componentLinearAccScale {z=1}
    componentLinearVelScale {z=0.5}
    gravityOverride {z=-980}
    physicsBodyDefinitions [{
      sphereCollisionRadius=0
      boundBone {boneName=belt2_r}
      boxExtents=<80 30 30>
      constraintSetup {angularLimitsMax=<0 0 5>, angularLimitsMin=<0 0 -10>, angularTarget=<1 0 0>, linearAxesMin=<-3 0 0>}
      localJointOffset=<2 0 0>
     }]
   }
   n4 = Anim Dynamics - Bone: belt2_l(ComponentPose=@n3, Alpha=1, ExternalForce="0,0,0") {
    angularDampingOverride=0.999
    angularSpringConstant=15000
    bGravityOverrideInSimSpace=true
    bLinearSpring=true
    bOverrideAngularDamping=true
    gravityScale=1.01
    linearSpringConstant=70
    numSolverIterationsPostUpdate=4
    numSolverIterationsPreUpdate=16
    simulationSpace=RootRelative
    alphaBoolBlend {blendOption=Linear}
    boundBone {boneName=belt2_l}
    componentAppliedLinearAccClamp=<3000 3000 3000>
    componentLinearAccScale {z=1}
    componentLinearVelScale {z=0.5}
    gravityOverride {z=-980}
    physicsBodyDefinitions [{
      sphereCollisionRadius=0
      boundBone {boneName=belt2_l}
      boxExtents=<85 35 35>
      constraintSetup {angularLimitsMax=<0 0 5>, angularLimitsMin=<0 0 -5>, angularTarget=<1 0 0>, linearAxesMax=<3.5 0 0>}
      localJointOffset=<-1.5 0 0>
     }]
   }
   n5 = Anim Dynamics - Bone: belt1_l(ComponentPose=@n4, Alpha=1, ExternalForce="0,0,0") {
    angularDampingOverride=0.999
    angularSpringConstant=15000
    bGravityOverrideInSimSpace=true
    bLinearSpring=true
    bOverrideAngularDamping=true
    gravityScale=1.01
    linearSpringConstant=70
    numSolverIterationsPostUpdate=4
    numSolverIterationsPreUpdate=16
    simulationSpace=RootRelative
    alphaBoolBlend {blendOption=Linear}
    boundBone {boneName=belt1_l}
    componentAppliedLinearAccClamp=<3000 3000 3000>
    componentLinearAccScale {z=1}
    componentLinearVelScale {z=0.5}
    gravityOverride {z=-980}
    physicsBodyDefinitions [{
      sphereCollisionRadius=0
      boundBone {boneName=belt1_l}
      boxExtents=<85 35 35>
      constraintSetup {angularLimitsMax=<0 0 5>, angularLimitsMin=<0 0 -5>, angularTarget=<1 0 0>, linearAxesMax=<4 0 0>}
      localJointOffset=<-1.5 0 0>
     }]
   }
   n6 = Anim Dynamics - Bone: belt1_r(ComponentPose=@n5, Alpha=1, ExternalForce="0,0,0") {
    angularDampingOverride=0.999
    angularSpringConstant=15000
    bGravityOverrideInSimSpace=true
    bLinearSpring=true
    bOverrideAngularDamping=true
    gravityScale=1.01
    linearSpringConstant=70
    numSolverIterationsPostUpdate=4
    numSolverIterationsPreUpdate=16
    simulationSpace=RootRelative
    alphaBoolBlend {blendOption=Linear}
    boundBone {boneName=belt1_r}
    componentAppliedLinearAccClamp=<3000 3000 3000>
    componentLinearAccScale {z=1}
    componentLinearVelScale {z=0.5}
    gravityOverride {z=-980}
    physicsBodyDefinitions [{
      sphereCollisionRadius=0
      boundBone {boneName=belt1_r}
      boxExtents=<75 30 30>
      constraintSetup {angularLimitsMax=<0 0 5>, angularLimitsMin=<0 0 -5>, angularTarget=<1 0 0>, linearAxesMin=<-4 0 0>}
      localJointOffset=<-2 0 0>
     }]
   }
   n7 = Get thigh_rot_up_l() -> (thigh_rot_up_l:double)
   n8 = "Transform (Modify) Bone - Bone: dyn_beltbuckle"(ComponentPose=@n6, Alpha=@n7, Translation="-0.738,0.972,6.546", Rotation="37.122,10.996,-57.037", Scale="1,1,1") {
    rotationMode=BMM_Additive
    rotationSpace=BCS_BoneSpace
    translationMode=BMM_Additive
    translationSpace=BCS_BoneSpace
    alphaBoolBlend {blendOption=Linear}
    boneToModify {boneName=dyn_beltbuckle}
    rotation=<rot 37.122 10.9961 -57.0368>
    translation=<-0.738 0.972 6.546>
   }
   n9 = "Transform (Modify) Bone - Bone: belt1_l"(ComponentPose=@n8, Alpha=@n7, Translation="10.332,2.726,0.011", Rotation="-7.790,-100.230,-32.566", Scale="1,1,1") {
    rotationMode=BMM_Additive
    rotationSpace=BCS_BoneSpace
    translationMode=BMM_Additive
    translationSpace=BCS_BoneSpace
    alphaBoolBlend {blendOption=Linear}
    boneToModify {boneName=belt1_l}
    rotation=<rot -7.79 -100.2301 -32.5663>
    translation=<10.332 2.726 0.011>
   }
   n10 = "Transform (Modify) Bone - Bone: belt2_l"(ComponentPose=@n9, Alpha=@n7, Translation="5.977,-0.178,-2.051", Rotation="-26.841,-10.039,9.973", Scale="1,1,1") {
    rotationMode=BMM_Additive
    rotationSpace=BCS_BoneSpace
    translationMode=BMM_Additive
    translationSpace=BCS_BoneSpace
    alphaBoolBlend {blendOption=Linear}
    boneToModify {boneName=belt2_l}
    rotation=<rot -26.841 -10.0393 9.9733>
    translation=<5.977 -0.178 -2.051>
   }
   n11 = "Transform (Modify) Bone - Bone: dyn_belt3"(ComponentPose=@n10, Alpha=1, Translation="0,0,0", Rotation="8.599,0,0.000", Scale="1,1,1") {
    rotationMode=BMM_Additive
    rotationSpace=BCS_BoneSpace
    translationMode=BMM_Additive
    translationSpace=BCS_BoneSpace
    alphaBoolBlend {blendOption=Linear}
    boneToModify {boneName=dyn_belt3}
    rotation {pitch=8.599}
   }
   n12 = "Transform (Modify) Bone - Bone: belt1_r"(ComponentPose=@n11, Alpha=@n7, Translation="-1.420,-0.597,-0.514", Rotation="-22.641,-8.235,14.041", Scale="1,1,1") {
    rotationMode=BMM_Additive
    rotationSpace=BCS_BoneSpace
    translationMode=BMM_Additive
    translationSpace=BCS_BoneSpace
    alphaBoolBlend {blendOption=Linear}
    boneToModify {boneName=belt1_r}
    rotation=<rot -22.641 -8.2351 14.0412>
    translation=<-1.42 -0.597 -0.514>
   }
   n13 = Get thigh_rot_up_r() -> (thigh_rot_up_r:double)
   n14 = "Transform (Modify) Bone - Bone: dyn_beltbuckle"(ComponentPose=@n12, Alpha=@n13, Translation="0.854,2.099,5.676", Rotation="-32.557,-19.478,-56.685", Scale="1,1,1") {
    rotationMode=BMM_Additive
    rotationSpace=BCS_BoneSpace
    translationMode=BMM_Additive
    translationSpace=BCS_BoneSpace
    alphaBoolBlend {blendOption=Linear}
    boneToModify {boneName=dyn_beltbuckle}
    rotation=<rot -32.557 -19.478 -56.6849>
    translation=<0.854 2.099 5.676>
   }
   n15 = "Transform (Modify) Bone - Bone: belt1_r"(ComponentPose=@n14, Alpha=@n13, Translation="-11.436,-4.923,-2.000", Rotation="0.000,-76.819,0.000", Scale="1,1,1") {
    rotationMode=BMM_Additive
    rotationSpace=BCS_BoneSpace
    translationMode=BMM_Additive
    translationSpace=BCS_BoneSpace
    alphaBoolBlend {blendOption=Linear}
    boneToModify {boneName=belt1_r}
    rotation {yaw=-76.819}
    translation=<-11.436 -4.923 -2>
   }
   n16 = "Transform (Modify) Bone - Bone: belt2_r"(ComponentPose=@n15, Alpha=@n13, Translation="-1.713,-1.597,-0.000", Rotation="0.000,-35.964,0.000", Scale="1,1,1") {
    rotationMode=BMM_Additive
    rotationSpace=BCS_BoneSpace
    translationMode=BMM_Additive
    translationSpace=BCS_BoneSpace
    alphaBoolBlend {blendOption=Linear}
    boneToModify {boneName=belt2_r}
    rotation {yaw=-35.964}
    translation=<-1.713 -1.597>
   }
   n17 = "Constraint - thigh_volume_l (R)"(ComponentPose=@n16, Alpha=0.5, ConstraintWeights_0=1) {
    alphaBoolBlend {blendOption=Linear}
    boneToModify {boneName=thigh_volume_l}
    constraintSetup [{transformType=Rotation, targetBone {boneName=thigh_l}}]
    constraintWeights [1]
   }
   n18 = "Transform (Modify) Bone - Bone: thigh_volume_l"(ComponentPose=@n17, Alpha=Get thigh_rot_up_l(), Translation="-1,-2,1", Rotation="0,0,0", Scale="1,1,1") {
    translationMode=BMM_Additive
    translationSpace=BCS_BoneSpace
    alphaBoolBlend {blendOption=Linear}
    boneToModify {boneName=thigh_volume_l}
    translation=<-1 -2 1>
   }
   n19 = "Constraint - thigh_volume_r (R)"(ComponentPose=@n18, Alpha=0.5, ConstraintWeights_0=1) {
    alphaBoolBlend {blendOption=Linear}
    boneToModify {boneName=thigh_volume_r}
    constraintSetup [{transformType=Rotation, targetBone {boneName=thigh_r}}]
    constraintWeights [1]
   }
   n20 = "Transform (Modify) Bone - Bone: thigh_volume_r"(ComponentPose=@n19, Alpha=Get thigh_rot_up_r(), Translation="1,2,-1", Rotation="0,0,0", Scale="1,1,1") {
    translationMode=BMM_Additive
    translationSpace=BCS_BoneSpace
    alphaBoolBlend {blendOption=Linear}
    boneToModify {boneName=thigh_volume_r}
    translation=<1 2 -1>
   }
   n21 = "Transform (Modify) Bone - Bone: scarf_back"(
    ComponentPose=@n20
    Alpha=Get scarf_back_alpha() -> (scarf_back_alpha:double)
    Translation="0,0,0"
    Rotation="25.963,-50.864,0.000"
    Scale="1,1,1"
   ) {
    rotationMode=BMM_Additive
    rotationSpace=BCS_BoneSpace
    translationSpace=BCS_BoneSpace
    alphaBoolBlend {blendOption=Linear}
    boneToModify {boneName=scarf_back}
    rotation=<rot 25.963 -50.8642 0>
   }
   Output Pose(Result=Component To Local(ComponentPose=@n21)) {layerGroup=None}
  ]}
 EventGraph {nodes [
   Event BlueprintUpdateAnimation()
   > n1 = Sequence!(execute) -> (then_0, then_1, then_2:Exec)
   > n2 = Branch(Condition=Get bIsCrouching())
   > n3 = Branch(Condition="float > float"(A=Get PawnSpeed2D(), B=60) {is=title K2Node_CallFunction})
   > Set belt_alpha(belt_alpha:double=1) -> (Output_Get:double)
   Set belt_alpha(execute=@n3.else, belt_alpha=0.15)
   Set belt_alpha(execute=@n2.else, belt_alpha=1)
   n4 = Get skel_mesh() -> (skel_mesh:Skeletal Mesh Component Object Reference)
   n5 = BreakRotator(InRot=BreakTransform(InTransform=GetDeltaTransformFromRefPose(self=@n4, BoneName=head, BaseName=spine_05)).Rotation)
   Set scarf_front_alpha(execute=@n1.then_2, scarf_front_alpha:double=MapRangeClamped(Value=@n5.Yaw, InRangeA=0, InRangeB=20)) -> (Output_Get:double)
   > Set scarf_back_alpha(scarf_back_alpha:double=MapRangeClamped(Value=@n5.Yaw, InRangeA=-10, InRangeB=-45)) -> (Output_Get:double)
   n6 = Get skel_mesh()
   n7 = BreakTransform(InTransform=GetDeltaTransformFromRefPose(self=@n6, BoneName=thigh_r, BaseName=pelvis))
   Set thigh_rot_up_r(
    execute=@n1.then_1
    thigh_rot_up_r:double=MapRangeClamped(Value=BreakRotator(InRot=@n7.Rotation).Yaw, InRangeA=-30, InRangeB=-100, OutRangeB=0.5)
   ) -> (Output_Get:double)
   n8 = BreakTransform(InTransform=GetDeltaTransformFromRefPose(self=@n6, BoneName=thigh_l, BaseName=pelvis))
   > Set thigh_rot_up_l(thigh_rot_up_l:double=MapRangeClamped(Value=BreakRotator(InRot=@n8.Rotation).Yaw, InRangeA=-10, InRangeB=-101)) -> (Output_Get:double)
   Event BlueprintInitializeAnimation()
   > Set skel_mesh(skel_mesh:Skeletal Mesh Component Object Reference=GetOwnerSkeletalMeshComponent()) -> (Output_Get:Skeletal Mesh Component Object Reference)
  ]}
]
variables [
 belt_alpha {type=number, default=1}
 scarf_back_alpha {type=number}
 scarf_front_alpha {type=number}
 skel_mesh {type=SkeletalMeshComponent}
 thigh_rot_up_l {type=number}
 thigh_rot_up_r {type=number}
]
