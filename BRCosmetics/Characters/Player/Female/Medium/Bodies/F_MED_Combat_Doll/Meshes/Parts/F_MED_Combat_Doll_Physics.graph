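# Physics asset for the hair of the female medium Combat Doll player cosmetic, covering bangs and three hair strands.
# Each dyn_ bone has a capsule body, with a larger capsule on the head as the root every chain attaches to.
# Center, left and right bang chains and three three-segment hair chains link bone to bone through constraints.
# Limits are mostly tight swing and twist, with the center bang allowed noticeably wider swing.
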PhysicsAsset /BRCosmetics/Characters/Player/Female/Medium/Bodies/F_MED_Combat_Doll/Meshes/Parts/F_MED_Combat_Doll_Physics
bodies {
 dyn_bang_c_01 [{shapeType=Capsule, length=2.0077, radius=1.3377, center=<0.1037 0 0.83>, rotation=<rot 0 -89.9997 -82.8756>}]
 dyn_bang_c_02 [{shapeType=Capsule, length=2.9622, radius=1.5086, center=<2.4563 -0.3285 0.3621>, rotation=<rot -5.1437 -97.8935 105.0097>}]
 dyn_bang_l_01 [{shapeType=Capsule, length=2.2159, radius=1.7168, center=<1.7106 -0.0001 0.3783>, rotation=<rot 90 0 0>}]
 dyn_bang_l_02 [{shapeType=Capsule, length=2.9622, radius=1.8256, center=<2.0369 0.022 0.3714>, rotation=<rot -79.7643 0 0>}]
 dyn_bang_l_03 [{shapeType=Capsule, length=2.7426, radius=1.8115, center=<0.4438 0.0833 0.3555>, rotation=<rot 0 -73.26 -80.8857>}]
 dyn_bang_r_01 [{shapeType=Capsule, length=2.9622, radius=1.5086, rotation=<rot 0 -89.9999 62.0374>}]
 dyn_bang_r_02 [{shapeType=Capsule, length=2.9622, radius=1.5086, center=<0.7637 0 -0.1452>, rotation=<rot 0 -89.9995 81.1414>}]
 dyn_bang_r_03 [{shapeType=Capsule, length=4.5, radius=1.6353, center=<-1.5266 0.2857 -0.5332>, rotation=<rot 90 0 0>}]
 dyn_hair1_01 [{shapeType=Capsule, length=2, radius=1.5, center=<0 0.6044 0>, rotation=<rot 90 0 0>}]
 dyn_hair1_02 [{shapeType=Capsule, length=2.4463, radius=1.8712, center=<0 0.6754 0>, rotation=<rot 90 0 0>}]
 dyn_hair1_03 [{shapeType=Capsule, length=2.4463, radius=1.8712, rotation=<rot 90 0 0>}]
 dyn_hair2_01 [{shapeType=Capsule, length=2, radius=1.5, center=<0 0.4527 0>, rotation=<rot 90 0 0>}]
 dyn_hair2_02 [{shapeType=Capsule, length=2.4463, radius=1.8712, center=<0 0.2392 0.6886>, rotation=<rot 90 0 0>}]
 dyn_hair2_03 [{shapeType=Capsule, length=2.4463, radius=1.8712, rotation=<rot 90 0 0>}]
 dyn_hair3_01 [{shapeType=Capsule, length=2, radius=1.5, center=<0 1.492 0>, rotation=<rot 90 0 0>}]
 dyn_hair3_02 [{shapeType=Capsule, length=2.4463, radius=1.8712, center=<0 0.4917 0>, rotation=<rot 90 0 0>}]
 dyn_hair3_03 [{shapeType=Capsule, length=2.4463, radius=1.8712, center=<0 0.0359 0>, rotation=<rot 90 0 0>}]
 head [{shapeType=Capsule, length=4.5, radius=6.2388, rotation=<rot 90 0 0>}]
}
constraints [
 {bone1Name=dyn_bang_c_01, bone2Name=head, swing1LimitDegrees=5, swing2LimitDegrees=45, twistLimitDegrees=5}
 {bone1Name=dyn_bang_c_02, bone2Name=dyn_bang_c_01, swing1LimitDegrees=45, swing2LimitDegrees=45, twistLimitDegrees=5}
 {bone1Name=dyn_bang_l_01, bone2Name=head, swing1LimitDegrees=2, swing2LimitDegrees=2, twistLimitDegrees=5}
 {bone1Name=dyn_bang_l_02, bone2Name=dyn_bang_l_01, swing1LimitDegrees=4, swing2LimitDegrees=12, twistLimitDegrees=5}
 {bone1Name=dyn_bang_l_03, bone2Name=dyn_bang_l_02, swing1LimitDegrees=4, swing2LimitDegrees=12, twistLimitDegrees=5}
 {bone1Name=dyn_bang_r_01, bone2Name=head, swing1LimitDegrees=0, swing2LimitDegrees=10, twistLimitDegrees=5}
 {bone1Name=dyn_bang_r_02, bone2Name=dyn_bang_r_01, swing1LimitDegrees=5, swing2LimitDegrees=10, twistLimitDegrees=5}
 {bone1Name=dyn_bang_r_03, bone2Name=dyn_bang_r_02, swing1LimitDegrees=5, swing2LimitDegrees=10, twistLimitDegrees=5}
 {bone1Name=dyn_hair1_01, bone2Name=head, swing1LimitDegrees=5, swing2LimitDegrees=5, twistLimitDegrees=5}
 {bone1Name=dyn_hair1_02, bone2Name=dyn_hair1_01, swing1LimitDegrees=5, swing2LimitDegrees=5, twistLimitDegrees=5}
 {bone1Name=dyn_hair1_03, bone2Name=dyn_hair1_02, swing1LimitDegrees=5, swing2LimitDegrees=5, twistLimitDegrees=5}
 {bone1Name=dyn_hair2_01, bone2Name=head, swing1LimitDegrees=5, swing2LimitDegrees=5, twistLimitDegrees=5}
 {bone1Name=dyn_hair2_02, bone2Name=dyn_hair2_01, swing1LimitDegrees=5, swing2LimitDegrees=5, twistLimitDegrees=5}
 {bone1Name=dyn_hair2_03, bone2Name=dyn_hair2_02, swing1LimitDegrees=5, swing2LimitDegrees=5, twistLimitDegrees=5}
 {bone1Name=dyn_hair3_01, bone2Name=head, swing1LimitDegrees=5, swing2LimitDegrees=5, twistLimitDegrees=5}
 {bone1Name=dyn_hair3_02, bone2Name=dyn_hair3_01, swing1LimitDegrees=5, swing2LimitDegrees=5, twistLimitDegrees=5}
 {bone1Name=dyn_hair3_03, bone2Name=dyn_hair3_02, swing1LimitDegrees=5, swing2LimitDegrees=5, twistLimitDegrees=5}
]
